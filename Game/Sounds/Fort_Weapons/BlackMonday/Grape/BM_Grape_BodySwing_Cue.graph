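# Sound cue for the body swing sound of the Black Monday "Grape" item, classed under gadget items.
# A mixer layers an immediate, louder random body-pull sound with a quieter random body-swing sound on a short fixed delay.
# Each layer gets slight volume variation from its own modulator.
# It uses grappling-gadget attenuation with local overrides that disable reverb send, plus player melee weapon concurrency.
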SoundCue /Game/Sounds/Fort_Weapons/BlackMonday/Grape/BM_Grape_BodySwing_Cue
root=Mixer_0
nodes [
 {id=Delay_0, children [Modulator_2], properties {delayMax=0.3, delayMin=0.3}}
 {id=Mixer_0, children [Delay_0,Modulator_3], properties {inputVolume [0.45,1.2]}}
 {id=Modulator_2, children [Random_2], properties {volumeMax=1.1, volumeMin=1}}
 {id=Modulator_3, children [Random_3], properties {volumeMax=0.95, volumeMin=0.85}}
 {id=Random_2, children [WavePlayer_4,WavePlayer_5,WavePlayer_6], properties {weights [1,1,1]}}
 {id=Random_3, children [WavePlayer_13, WavePlayer_14], properties {weights [1,1]}}
 {id=WavePlayer_13, properties {soundWaveAssetPtr=&/BM_Grape_BodyPull_02}}
 {id=WavePlayer_14, properties {soundWaveAssetPtr=&/BM_Grape_BodyPull_03}}
 {id=WavePlayer_4, properties {soundWaveAssetPtr=&/BM_Grape_BodySwing_01}}
 {id=WavePlayer_5, properties {soundWaveAssetPtr=&/BM_Grape_BodySwing_02}}
 {id=WavePlayer_6, properties {soundWaveAssetPtr=&/BM_Grape_BodySwing_03}}
]
properties {
 attenuationSettings=/Game/Sounds/Attenuation/Fort_GadgetGrappling_Att
 bOverrideAttenuation=true
 maxDistance=3000
 soundClassObject=&Fort_Gadgets_Items
 volumeMultiplier=1.2
 attenuationOverrides {bEnableReverbSend=false, falloffDistance=3000, nonSpatializedRadiusStart=800, stereoSpread=2000, attenuationShapeExtents {x=0}}
 concurrencySet [&Player_Melee_Weapon_Conc]
}
