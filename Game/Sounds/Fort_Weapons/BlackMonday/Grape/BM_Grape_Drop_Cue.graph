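# Sound cue for the drop sound of the BlackMonday "Grape" item, filed under weapon sounds.
# A Random node picks one of three drop waves with equal weight, and a Modulator adds slight volume variation with no pitch change.
# It uses the weapon drop attenuation, the loot collect sound class and the shared Weapon_Drop_StopOldest concurrency set.
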SoundCue /Game/Sounds/Fort_Weapons/BlackMonday/Grape/BM_Grape_Drop_Cue
root=Modulator_1
nodes [
 {id=Modulator_1, children [Random_0], properties {pitchMax=1, pitchMin=1, volumeMax=0.8, volumeMin=0.7}}
 {id=Random_0, children [WavePlayer_0,WavePlayer_2,WavePlayer_3], properties {weights [1,1,1]}}
 {id=WavePlayer_0, properties {soundWaveAssetPtr=&/BM_Grape_Drop_01}}
 {id=WavePlayer_2, properties {soundWaveAssetPtr=&/BM_Grape_Drop_02}}
 {id=WavePlayer_3, properties {soundWaveAssetPtr=&/BM_Grape_Drop_03}}
]
properties {
 attenuationSettings=&Fort_Weapon_Drop_Att
 maxDistance=3000
 soundClassObject=&Fort_Loot_Collect
 volumeMultiplier=0.6892
 concurrencyOverrides {maxCount=1, resolutionRule=PreventNew}
 concurrencySet [&Weapon_Drop_StopOldest]
}
