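# First-person fire sound cue for the BlackMonday Grape weapon, layering a close shot with a cable sound.
# A quality-level root scales variation: the top tier randomizes four waves per layer, the next uses two, and the last two share a fixed pair.
# Only the close layer in the top two tiers passes through a modulator.
# It uses first-person weapon attenuation and gunfire concurrency, with a PreventNew override also set.
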SoundCue /Game/Sounds/Fort_Weapons/BlackMonday/Grape/BM_Grape_Fire_1P_Cue
root=QualityLevel_0
nodes [
 {id=Mixer_0, children [Modulator_1,Random_5], properties {inputVolume [0.75,0.65]}}
 {id=Mixer_1, children [Mixer_2,null], properties {inputVolume [0.8,0.8]}}
 {id=Mixer_2, children [Modulator_0,Random_3], properties {inputVolume [0.95,0.85]}}
 {id=Mixer_3, children [WavePlayer_12, WavePlayer_22], properties {inputVolume [0.75,0.65]}}
 {id=Modulator_0, children [Random_0]}
 {id=Modulator_1, children [Random_2]}
 {id=QualityLevel_0, children [Mixer_1,Mixer_0,Mixer_3,Mixer_3]}
 {id=Random_0, children [WavePlayer_0,WavePlayer_1,WavePlayer_6,WavePlayer_7], properties {weights [1,1,1,1]}}
 {id=Random_2, children [WavePlayer_8,WavePlayer_9], properties {weights [1,1]}}
 {id=Random_3, children [WavePlayer_13, WavePlayer_14, WavePlayer_18, WavePlayer_19], properties {weights [1,1,1,1]}}
 {id=Random_5, children [WavePlayer_20, WavePlayer_21], properties {weights [1,1]}}
 {id=WavePlayer_0, properties {soundWaveAssetPtr=&/BM_Grape_Fire_Close_01}}
 {id=WavePlayer_1, properties {soundWaveAssetPtr=&/BM_Grape_Fire_Close_02}}
 {id=WavePlayer_12, properties {soundWaveAssetPtr=&BM_Grape_Fire_Close_01}}
 {id=WavePlayer_13, properties {soundWaveAssetPtr=&/BM_Grape_Fire_Cable_01}}
 {id=WavePlayer_14, properties {soundWaveAssetPtr=&/BM_Grape_Fire_Cable_02}}
 {id=WavePlayer_18, properties {soundWaveAssetPtr=&/BM_Grape_Fire_Cable_03}}
 {id=WavePlayer_19, properties {soundWaveAssetPtr=&/BM_Grape_Fire_Cable_04}}
 {id=WavePlayer_20, properties {soundWaveAssetPtr=&BM_Grape_Fire_Cable_01}}
 {id=WavePlayer_21, properties {soundWaveAssetPtr=&BM_Grape_Fire_Cable_02}}
 {id=WavePlayer_22, properties {soundWaveAssetPtr=&BM_Grape_Fire_Cable_01}}
 {id=WavePlayer_6, properties {soundWaveAssetPtr=&/BM_Grape_Fire_Close_03}}
 {id=WavePlayer_7, properties {soundWaveAssetPtr=&/BM_Grape_Fire_Close_04}}
 {id=WavePlayer_8, properties {soundWaveAssetPtr=&BM_Grape_Fire_Close_01}}
 {id=WavePlayer_9, properties {soundWaveAssetPtr=&BM_Grape_Fire_Close_02}}
]
properties {
 attenuationSettings=&Fort_Weapon_Projectile_FirstPerson_Athena_Att
 maxDistance=2097152
 soundClassObject=&Fort_WeaponRanged_Fire_1P_PSM
 volumeMultiplier=0.81
 attenuationOverrides {falloffDistance=5000, attenuationShapeExtents {x=1000}}
 concurrencyOverrides {resolutionRule=PreventNew}
 concurrencySet [&Weapon_Gunfire_FirstPerson_Conc]
}
