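# Sound cue for firing the BlackMonday Grape weapon, with a third-person close layer and a distant layer.
# A quality-level switch selects branches with four, two or one wave variations per layer, the last two levels sharing the single-variation branch.
# Each branch mixes randomized Close plus Cable waves with Medium waves, each path getting its own modulation, attenuation and sound class.
# Local and global gunfire concurrency sets limit playback.
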SoundCue /Game/Sounds/Fort_Weapons/BlackMonday/Grape/BM_Grape_Fire_Cue
root=QualityLevel_0
nodes [
 {id=Attenuation_0, children [Modulator_1], properties {attenuationSettings=&Fort_Weapon_Projectile_Close_Athena_Att}}
 {
  id=Attenuation_1
  children [Modulator_4]
  properties {attenuationSettings=&Fort_Weapon_Projectile_Medium_Mono_Athena_Att}
 }
 {id=Attenuation_2, children [Modulator_2], properties {attenuationSettings=&Fort_Weapon_Projectile_Close_Athena_Att}}
 {id=Attenuation_3, children [Modulator_0], properties {attenuationSettings=&Fort_Weapon_Projectile_Close_Athena_Att}}
 {
  id=Attenuation_4
  children [Modulator_3]
  properties {attenuationSettings=&Fort_Weapon_Projectile_Medium_Mono_Athena_Att}
 }
 {
  id=Attenuation_9
  children [Modulator_5]
  properties {attenuationSettings=&Fort_Weapon_Projectile_Medium_Mono_Athena_Att}
 }
 {id=Mixer_0, children [SoundClass_1,SoundClass_0], properties {inputVolume [0.8,0.8]}}
 {id=Mixer_1, children [SoundClass_3,SoundClass_2], properties {inputVolume [0.8,0.8]}}
 {id=Mixer_2, children [Random_0,Random_8], properties {inputVolume [0.95,0.85]}}
 {id=Mixer_3, children [SoundClass_7,SoundClass_6], properties {inputVolume [0.8,0.8]}}
 {id=Mixer_4, children [Random_3,Random_9], properties {inputVolume [0.95,0.85]}}
 {id=Mixer_5, children [WavePlayer_21, WavePlayer_34], properties {inputVolume [0.95,0.85]}}
 {id=Modulator_0, children [Mixer_2], properties {volumeMax=0.85, volumeMin=0.75}}
 {id=Modulator_1, children [Mixer_4], properties {volumeMax=0.85, volumeMin=0.75}}
 {id=Modulator_2, children [Mixer_5], properties {volumeMax=0.85, volumeMin=0.75}}
 {id=Modulator_3, children [Random_5]}
 {id=Modulator_4, children [Random_7]}
 {id=Modulator_5, children [WavePlayer_31]}
 {id=QualityLevel_0, children [Mixer_1,Mixer_0,Mixer_3,Mixer_3]}
 {id=Random_0, children [WavePlayer_0, WavePlayer_1, WavePlayer_7, WavePlayer_10], properties {weights [1,1,1,1]}}
 {id=Random_3, children [WavePlayer_11, WavePlayer_18], properties {weights [1,1]}}
 {id=Random_5, children [WavePlayer_22, WavePlayer_23, WavePlayer_24, WavePlayer_25], properties {weights [1,1,1,1]}}
 {id=Random_7, children [WavePlayer_27, WavePlayer_28], properties {weights [1,1]}}
 {id=Random_8, children [WavePlayer_13, WavePlayer_14, WavePlayer_32, WavePlayer_19], properties {weights [1,1,1,1]}}
 {id=Random_9, children [WavePlayer_20, WavePlayer_33], properties {weights [1,1]}}
 {
  id=SoundClass_0
  children [Attenuation_1]
  properties {soundClassOverride=&Fort_WeaponRanged_Fire_Distant}
 }
 {
  id=SoundClass_1
  children [Attenuation_0]
  properties {soundClassOverride=&Fort_WeaponRanged_Fire_3P_PSM}
 }
 {
  id=SoundClass_2
  children [Attenuation_4]
  properties {soundClassOverride=&Fort_WeaponRanged_Fire_Distant}
 }
 {
  id=SoundClass_3
  children [Attenuation_3]
  properties {soundClassOverride=&Fort_WeaponRanged_Fire_3P_PSM}
 }
 {
  id=SoundClass_6
  children [Attenuation_9]
  properties {soundClassOverride=&Fort_WeaponRanged_Fire_Distant}
 }
 {
  id=SoundClass_7
  children [Attenuation_2]
  properties {soundClassOverride=&Fort_WeaponRanged_Fire_3P_PSM}
 }
 {id=WavePlayer_0, properties {soundWaveAssetPtr=&/BM_Grape_Fire_Close_01}}
 {id=WavePlayer_1, properties {soundWaveAssetPtr=&/BM_Grape_Fire_Close_02}}
 {id=WavePlayer_10, properties {soundWaveAssetPtr=&/BM_Grape_Fire_Close_04}}
 {id=WavePlayer_11, properties {soundWaveAssetPtr=&BM_Grape_Fire_Close_01}}
 {id=WavePlayer_13, properties {soundWaveAssetPtr=&/BM_Grape_Fire_Cable_01}}
 {id=WavePlayer_14, properties {soundWaveAssetPtr=&/BM_Grape_Fire_Cable_02}}
 {id=WavePlayer_18, properties {soundWaveAssetPtr=&BM_Grape_Fire_Close_04}}
 {id=WavePlayer_19, properties {soundWaveAssetPtr=&/BM_Grape_Fire_Cable_04}}
 {id=WavePlayer_20, properties {soundWaveAssetPtr=&BM_Grape_Fire_Cable_01}}
 {id=WavePlayer_21, properties {soundWaveAssetPtr=&BM_Grape_Fire_Close_01}}
 {id=WavePlayer_22, properties {soundWaveAssetPtr=&/BM_Grape_Fire_Medium_01}}
 {id=WavePlayer_23, properties {soundWaveAssetPtr=&/BM_Grape_Fire_Medium_02}}
 {id=WavePlayer_24, properties {soundWaveAssetPtr=&/BM_Grape_Fire_Medium_03}}
 {id=WavePlayer_25, properties {soundWaveAssetPtr=&/BM_Grape_Fire_Medium_04}}
 {id=WavePlayer_27, properties {soundWaveAssetPtr=&BM_Grape_Fire_Medium_01}}
 {id=WavePlayer_28, properties {soundWaveAssetPtr=&BM_Grape_Fire_Medium_02}}
 {id=WavePlayer_31, properties {soundWaveAssetPtr=&BM_Grape_Fire_Medium_01}}
 {id=WavePlayer_32, properties {soundWaveAssetPtr=&/BM_Grape_Fire_Cable_03}}
 {id=WavePlayer_33, properties {soundWaveAssetPtr=&BM_Grape_Fire_Cable_02}}
 {id=WavePlayer_34, properties {soundWaveAssetPtr=&BM_Grape_Fire_Cable_01}}
 {id=WavePlayer_7, properties {soundWaveAssetPtr=&/BM_Grape_Fire_Close_03}}
]
properties {
 maxDistance=10579.2871
 attenuationOverrides {falloffDistance=5000, attenuationShapeExtents {x=1000}}
 concurrencyOverrides {resolutionRule=PreventNew}
 concurrencySet [&Weapon_Gunfire_Local_Conc, &Weapon_Gunfire_Global_Conc]
}
