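# Sound cue for the Black Monday Grape flesh impact sound, assigned to the Fort_Impacts sound class.
# A short fixed delay feeds a default Modulator, then a Random node picks one of three equally weighted wave variants.
# Overridden attenuation turns off distance falloff, spatialization and reverb send, so the impact plays as non-positional audio.
# It shares the RifleBulletImpact_Conc concurrency set to manage overlapping impacts.
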SoundCue /Game/Sounds/Fort_Weapons/BlackMonday/Grape/BM_Grape_ImpactFlesh_Cue
root=Delay_0
nodes [
 {id=Delay_0, children [Modulator_1], properties {delayMax=0.05, delayMin=0.05}}
 {id=Modulator_1, children [Random_0]}
 {id=Random_0, children [WavePlayer_0,WavePlayer_1,WavePlayer_2], properties {weights [1,1,1]}}
 {id=WavePlayer_0, properties {soundWaveAssetPtr=&/BM_Grape_ImpactFlesh_01}}
 {id=WavePlayer_1, properties {soundWaveAssetPtr=&/BM_Grape_ImpactFlesh_02}}
 {id=WavePlayer_2, properties {soundWaveAssetPtr=&/BM_Grape_ImpactFlesh_03}}
]
properties {
 bOverrideAttenuation=true
 maxDistance=2097152
 soundClassObject=&Fort_Impacts
 volumeMultiplier=1.425
 attenuationOverrides {
  bAttenuate=false
  bEnableReverbSend=false
  bSpatialize=false
  falloffDistance=3000
  nonSpatializedRadiusStart=1000
  stereoSpread=2000
  attenuationShapeExtents {x=2500}
 }
 concurrencySet [&RifleBulletImpact_Conc]
}
