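# Sound cue for the BlackMonday Grape weapon's surface impact sound.
# An equally weighted Random node picks one of three impact waves and feeds a Modulator at default settings for slight variation.
# Playback uses the Fort_Impacts sound class, the rifle bullet impact concurrency set and a boosted volume.
# It references gadget-activate attenuation and carries attenuation overrides for falloff, stereo spread and no reverb send.
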SoundCue /Game/Sounds/Fort_Weapons/BlackMonday/Grape/BM_Grape_ImpactSurface_Cue
root=Modulator_1
nodes [
 {id=Modulator_1, children [Random_1]}
 {id=Random_1, children [WavePlayer_4,WavePlayer_5,WavePlayer_6], properties {weights [1,1,1]}}
 {id=WavePlayer_4, properties {soundWaveAssetPtr=&/BM_Grape_ImpactSurface_01}}
 {id=WavePlayer_5, properties {soundWaveAssetPtr=&/BM_Grape_ImpactSurface_02}}
 {id=WavePlayer_6, properties {soundWaveAssetPtr=&/BM_Grape_ImpactSurface_03}}
]
properties {
 attenuationSettings=&Fort_GadgetActivate_Att
 maxDistance=4000
 soundClassObject=&Fort_Impacts
 volumeMultiplier=1.6
 attenuationOverrides {bEnableReverbSend=false, falloffDistance=3000, nonSpatializedRadiusStart=1000, stereoSpread=2000, attenuationShapeExtents {x=2500}}
 concurrencySet [&RifleBulletImpact_Conc]
}
