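# Sound cue for picking up the BlackMonday Grape weapon, played in the loot collect sound class with weapon pickup attenuation.
# A mixer layers a fixed dual pistols pickup wave with one of two Grape pickup waves, chosen at random with equal weight.
# Only the Grape layer passes through a modulator, which lowers and slightly varies its volume.
# Concurrency is overridden to allow one instance at a time, preventing new plays while one is active.
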SoundCue /Game/Sounds/Fort_Weapons/BlackMonday/Grape/BM_Grape_Pickup_Cue
root=Mixer_0
nodes [
 {id=Mixer_0, children [WavePlayer_4,Modulator_0], properties {inputVolume [1,1]}}
 {id=Modulator_0, children [Random_0], properties {pitchMax=1, pitchMin=1, volumeMax=0.8, volumeMin=0.7}}
 {id=Random_0, children [WavePlayer_0,WavePlayer_1], properties {weights [1,1]}}
 {id=WavePlayer_0, properties {soundWaveAssetPtr=&/BM_Grape_Pickup_01}}
 {id=WavePlayer_1, properties {soundWaveAssetPtr=&/BM_Grape_Pickup_02}}
 {id=WavePlayer_4, properties {soundWaveAssetPtr=/Game/Sounds/Fort_Weapons/pistol/Dual/DualPistols_Pickup_01}}
]
properties {
 attenuationSettings=&Fort_Weapon_Pickup_Att
 bOverrideConcurrency=true
 maxDistance=3500
 soundClassObject=&Fort_Loot_Collect
 volumeMultiplier=1.0338
 concurrencyOverrides {maxCount=1, resolutionRule=PreventNew}
}
