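# Sound cue for the end of the reload on the BlackMonday Grape weapon.
# A single wave player reuses the dual pistols pickup sound, played quieter and pitched up.
# It routes to the weapon foley sound class with weapon reload attenuation and overrides concurrency.
# The debug flag is left on.
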SoundCue /Game/Sounds/Fort_Weapons/BlackMonday/Grape/BM_Grape_Reload_End_Cue
root=WavePlayer_4
nodes [{id=WavePlayer_4, properties {soundWaveAssetPtr=/Game/Sounds/Fort_Weapons/pistol/Dual/DualPistols_Pickup_01}}]
properties {
 attenuationSettings=&Fort_Weapon_Reload_Att
 bDebug=true
 bOverrideConcurrency=true
 maxDistance=2250
 pitchMultiplier=1.2
 soundClassObject=&Fort_WeaponFoley
 volumeMultiplier=0.401
}
