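# Sound cue for the retract sound of the BlackMonday Grape weapon, classed under gadget items.
# A Random node picks between two retract wave variations with equal weight, then a Modulator adds slight volume variation above unity.
# The cue lowers overall volume and uses the shared grappling gadget attenuation for distance falloff.
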SoundCue /Game/Sounds/Fort_Weapons/BlackMonday/Grape/BM_Grape_Retract_Cue
root=Modulator_0
nodes [
 {id=Modulator_0, children [Random_0], properties {volumeMax=1.3, volumeMin=1.2}}
 {id=Random_0, children [WavePlayer_0,WavePlayer_2], properties {weights [1,1]}}
 {id=WavePlayer_0, properties {soundWaveAssetPtr=&/BM_Grape_Retract_01}}
 {id=WavePlayer_2, properties {soundWaveAssetPtr=&/BM_Grape_Retract_02}}
]
properties {
 attenuationSettings=/Game/Sounds/Attenuation/Fort_GadgetGrappling_Att
 maxDistance=8500
 soundClassObject=&Fort_Gadgets_Items
 volumeMultiplier=0.45
}
