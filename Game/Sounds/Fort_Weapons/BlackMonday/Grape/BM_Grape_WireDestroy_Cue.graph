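# Sound cue for the Black Monday Grape gadget's wire-destroy event, playing a cable-tightening sound.
# A Random node picks one of four equally weighted cable-tighten waves, and a Modulator varies its volume slightly above the base.
# The cue uses grappling-gadget attenuation, the gadgets and items sound class, and a concurrency rule limiting it to two instances per owner, stopping the oldest.
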SoundCue /Game/Sounds/Fort_Weapons/BlackMonday/Grape/BM_Grape_WireDestroy_Cue
root=Modulator_2
nodes [
 {id=Modulator_2, children [Random_2], properties {volumeMax=1.3, volumeMin=1.2}}
 {id=Random_2, children [WavePlayer_2, WavePlayer_3, WavePlayer_11, WavePlayer_12], properties {weights [1,1,1,1]}}
 {id=WavePlayer_11, properties {soundWaveAssetPtr=&/BM_Grape_CableTightens_03}}
 {id=WavePlayer_12, properties {soundWaveAssetPtr=&/BM_Grape_CableTightens_04}}
 {id=WavePlayer_2, properties {soundWaveAssetPtr=&/BM_Grape_CableTightens_01}}
 {id=WavePlayer_3, properties {soundWaveAssetPtr=&/BM_Grape_CableTightens_02}}
]
properties {
 attenuationSettings=/Game/Sounds/Attenuation/Fort_GadgetGrappling_Att
 maxDistance=8500
 soundClassObject=&Fort_Gadgets_Items
 volumeMultiplier=0.9
 concurrencySet [/Game/Sounds/Concurrency_Rules/TwoMaximum_Owner_StopOldest]
}
