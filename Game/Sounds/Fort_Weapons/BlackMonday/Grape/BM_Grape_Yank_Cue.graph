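# BM_Grape_Yank_Cue is a sound cue for the yank sound of the BlackMonday Grape item, classed as a gadget/item sound.
# A Random node picks one of four equally weighted yank waves, and a Modulator adds slight volume variation.
# It uses grappling gadget attenuation, a reduced volume multiplier, and a per-owner concurrency rule that caps it at two, stopping the oldest.
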SoundCue /Game/Sounds/Fort_Weapons/BlackMonday/Grape/BM_Grape_Yank_Cue
root=Modulator_0
nodes [
 {id=Modulator_0, children [Random_1], properties {volumeMax=1.1, volumeMin=1}}
 {id=Random_1, children [WavePlayer_0,WavePlayer_1,WavePlayer_2,WavePlayer_3], properties {weights [1,1,1,1]}}
 {id=WavePlayer_0, properties {soundWaveAssetPtr=&/BM_Grape_Yank_01}}
 {id=WavePlayer_1, properties {soundWaveAssetPtr=&/BM_Grape_Yank_02}}
 {id=WavePlayer_2, properties {soundWaveAssetPtr=&/BM_Grape_Yank_03}}
 {id=WavePlayer_3, properties {soundWaveAssetPtr=&/BM_Grape_Yank_04}}
]
properties {
 attenuationSettings=/Game/Sounds/Attenuation/Fort_GadgetGrappling_Att
 maxDistance=8500
 soundClassObject=&Fort_Gadgets_Items
 volumeMultiplier=0.37
 concurrencySet [/Game/Sounds/Concurrency_Rules/TwoMaximum_Owner_StopOldest]
}
